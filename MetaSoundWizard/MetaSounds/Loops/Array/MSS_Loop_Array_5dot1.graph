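# MetaSound Source template for a 5.1 looping sound that plays a random pick from a Sounds wave array, which is empty by default.
# On Play triggers a Random Get, which passes the chosen wave and trigger to a 5.1 Wave Player feeding the six output channels.
# The source is flagged looping, but the player's Loop default holds a stringified builder result instead of a boolean.
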MetaSoundSource /MetaSoundWizard/MetaSounds/Loops/Array/MSS_Loop_Array_5dot1
inputs [
 Sounds {type=WaveAsset:Array}
 UE.Source.OnPlay {type=Trigger}
]
interfaces [UE.OutputFormat.5dot1:1.0, UE.Source:1.1]
outputs [
 "UE.OutputFormat.5dot1.Audio:0" {type=Audio}
 "UE.OutputFormat.5dot1.Audio:1" {type=Audio}
 "UE.OutputFormat.5dot1.Audio:2" {type=Audio}
 "UE.OutputFormat.5dot1.Audio:3" {type=Audio}
 "UE.OutputFormat.5dot1.Audio:4" {type=Audio}
 "UE.OutputFormat.5dot1.Audio:5" {type=Audio}
]
pages [Default {
  input_defaults {UE.Source.OnPlay=null, Sounds []}
  nodes [
   ArrayRandomGet_0 = Array.Random Get:WaveAsset:Array(In Array=^Sounds, Next=^UE.Source.OnPlay)
   WavePlayer_0 = UE.Wave Player:5dot1(Play=@ArrayRandomGet_0.On Next, Wave Asset=@ArrayRandomGet_0.Value) {also {defaults {Loop="<MetaSoundBuilderResult.SUCCEEDED: 0>"}}}
  ]
  links [
   {dst="@UE.OutputFormat.5dot1.Audio:2", src=WavePlayer_0.Out Front Center}
   {dst="@UE.OutputFormat.5dot1.Audio:0", src=WavePlayer_0.Out Front Left}
   {dst="@UE.OutputFormat.5dot1.Audio:1", src=WavePlayer_0.Out Front Right}
   {dst="@UE.OutputFormat.5dot1.Audio:3", src=WavePlayer_0.Out Low Frequency}
   {dst="@UE.OutputFormat.5dot1.Audio:4", src=WavePlayer_0.Out Side Left}
   {dst="@UE.OutputFormat.5dot1.Audio:5", src=WavePlayer_0.Out Side Right}
  ]
 }]
properties {
 assetImportData=&self:AssetImportData
 bLooping=true
 duration=10000
 numChannels=6
 outputFormat=FiveDotOne
}
