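# A stereo MetaSound Source template that plays a random sound from the Sounds WaveAsset array, which is empty by default.
# On Play triggers a Random Get, which passes the chosen wave to a stereo Wave Player feeding the left and right outputs.
# The source is flagged as looping, but the player's Loop default holds a stringified MetaSound builder result rather than a plain boolean.
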MetaSoundSource /MetaSoundWizard/MetaSounds/Loops/Array/MSS_Loop_Array_Stereo
inputs [
 Sounds {type=WaveAsset:Array}
 UE.Source.OnPlay {type=Trigger}
]
interfaces [UE.OutputFormat.Stereo:1.0, UE.Source:1.1]
outputs [
 "UE.OutputFormat.Stereo.Audio:0" {type=Audio}
 "UE.OutputFormat.Stereo.Audio:1" {type=Audio}
]
pages [Default {
  input_defaults {UE.Source.OnPlay=null, Sounds []}
  nodes [
   ArrayRandomGet_0 = Array.Random Get:WaveAsset:Array(In Array=^Sounds, Next=^UE.Source.OnPlay)
   WavePlayer_0 = UE.Wave Player:Stereo(Play=@ArrayRandomGet_0.On Next, Wave Asset=@ArrayRandomGet_0.Value) {also {defaults {Loop="<MetaSoundBuilderResult.SUCCEEDED: 0>"}}}
  ]
  links [{dst="@UE.OutputFormat.Stereo.Audio:0", src=WavePlayer_0.Out Left}, {dst="@UE.OutputFormat.Stereo.Audio:1", src=WavePlayer_0.Out Right}]
 }]
properties {
 assetImportData=&self:AssetImportData
 bLooping=true
 duration=10000
 numChannels=2
 outputFormat=Stereo
}
